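# Sound cue for the Demon Arena pre-spawn energy audio, combining a start sound, a sustained loop and a later build-up.
# A mixer plays three attenuated branches: the start one-shot, a looping layer faded in by an enveloper at slightly reduced level, and a build one-shot delayed about half a minute.
# The one-shots use close POI attenuation, while the loop has its own POI loop attenuation.
# Concurrency is overridden to cap simultaneous instances.
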SoundCue /DemonArenaGameplay/Assets/Audio/SC_DemonArena_PreSpawn_Loop
root=Mixer_0
nodes [
 {id=Attenuation_0, children [Enveloper_0], properties {attenuationSettings=&/SA_DemonArena_POI_Loop}}
 {id=Attenuation_1, children [WavePlayer_2], properties {attenuationSettings=&/SA_DemonArena_POI_Close}}
 {id=Attenuation_2, children [Delay_1], properties {attenuationSettings=&SA_DemonArena_POI_Close}}
 {id=Delay_0, children [WavePlayer_1]}
 {id=Delay_1, children [WavePlayer_3], properties {delayMax=27.6, delayMin=27.6}}
 {
  id=Enveloper_0
  children [Delay_0]
  properties {pitchCurve {editorCurveData {keys [{time=0}]}}, volumeCurve {editorCurveData {keys [<at 0 0>,<at 1 1>]}}}
 }
 {id=Mixer_0, children [Attenuation_1, Attenuation_0, Attenuation_2], properties {inputVolume [1,0.7,1]}}
 {id=WavePlayer_1, properties {bLooping=true, soundWaveAssetPtr=&/DemonArena_PreSpawn_Energy_Loop_01}}
 {id=WavePlayer_2, properties {soundWaveAssetPtr=&/DemonArena_PreSpawn_Energy_Start_01}}
 {id=WavePlayer_3, properties {soundWaveAssetPtr=&/DemonArena_PreSpawn_Energy_Build_01}}
]
properties {
 bOverrideConcurrency=true
 maxDistance=120000
 soundClassObject=&Fort_Gadgets
 volumeMultiplier=1
 concurrencyOverrides {maxCount=2}
}
