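# Sound cue for the Demon Arena spawn sound, blending close and distant layers.
# A mixer combines a close wave and a distant wave, each through its own POI attenuation asset, with the close layer passing through an extra mixer that trims its level.
# The cue uses the Fort_Gadgets sound class and a very large max distance, so the spawn can carry far across the map.
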SoundCue /DemonArenaGameplay/Assets/Audio/SC_DemonArena_Spawn
root=Mixer_0
nodes [
 {id=Attenuation_0, children [WavePlayer_2], properties {attenuationSettings=&/SA_DemonArena_POI_Close}}
 {id=Attenuation_3, children [WavePlayer_0], properties {attenuationSettings=&/SA_DemonArena_POI_Distant}}
 {id=Mixer_0, children [Mixer_1, Attenuation_3], properties {inputVolume [1.2,0.8]}}
 {id=Mixer_1, children [Attenuation_0], properties {inputVolume [0.9]}}
 {id=WavePlayer_0, properties {soundWaveAssetPtr=&/DemonArena_Spawn_Dist_01}}
 {id=WavePlayer_2, properties {soundWaveAssetPtr=&/DemonArena_Spawn_Close_01}}
]
properties {maxDistance=120000, soundClassObject=&Fort_Gadgets, volumeMultiplier=1}
