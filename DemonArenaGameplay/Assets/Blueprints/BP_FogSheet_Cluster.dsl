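; BP_FogSheet_Cluster is a Blueprint that builds a cluster of fog sheets from one instanced static mesh component, FogPlane.
; Its construction script applies M_FogSheets, then adds an instance for each InstanceTransformList entry.
; Instances take only the location; each entry's X and Y scale go into per-instance custom data slots, presumably for the material to read.
; BeginPlay, Tick and overlap events are empty.
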(event EventBeginPlay)

(event Collision|EventActorBeginOverlap
  (params (OtherActor :type Actor%20Object%20Reference)))

(event EventTick
  (params (DeltaSeconds :type Float%20%28single-precision%29)))

(fn AddNewInstance
  (bind _fogplane (Variables|Default|GetFogPlane))
  (for _array_element (Variables|Default|GetInstanceTransformList)
    (bind _returnvalue (Components|InstancedStaticMesh|AddInstance _fogplane (Math|Transform|MakeTransform (.location _array_element)) false))
    (Components|InstancedStaticMesh|SetCustomDataValue _fogplane _returnvalue 0 (.x (.scale _array_element)) false)
    (Components|InstancedStaticMesh|SetCustomDataValue _fogplane _returnvalue 1 (.y (.scale _array_element)) false)))

(fn ConstructionScript
  (Rendering|Material|SetMaterial (Variables|Default|GetFogPlane) 0 "/DemonArenaGameplay/Assets/Materials/M_FogSheets.M_FogSheets")
  (CallFunction|AddNewInstance))
