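# Translucent unlit fog-sheet material for Demon Arena, drawn as camera-facing sprites that support instanced static meshes.
# World position offset sizes each sprite from per-instance or primitive data and pushes it toward the camera.
# Opacity combines UV gradient masks, cloud noise, time-of-day multipliers and depth fade, alongside camera-distance fade masks.
# Color blends fog tints through masks, including scene-depth and sun-direction transmittance masks, then mixes in skylight sampled along the sun vector.
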Material /DemonArenaGameplay/Assets/Materials/M_FogSheets
graph {
 properties {
  bAutomaticallySetUsageInEditor=false
  bUsedWithInstancedStaticMeshes=true
  blendMode=BLEND_Translucent
  shadingModel=MSM_Unlit
  usedShadingModels=Unlit
 }
 nodes [
  n1 = Divide(A=ObjectPositionWS(), B=VectorParameter() {parameterName=PositionOffsetBias, defaultValue=<rgba 5000 5000 5000 1>}.RGB)
  n2 = TextureCoordinate()
  Add(A=@n2, B=ComponentMask(None=Frac(None=@n1)) {g=true, r=true})
  Constant() {r=1}
  Constant() {r=1}
  n3 = NamedRerouteUsage()
  n4 = ComponentMask(None=VectorParameter() {parameterName=FogNoise_Tiling, defaultValue=<rgba 0.5 0.5 0 1>}.RGB) {g=true, r=true}
  n5 = Multiply(
   A=TextureSample(UVs=Multiply(A=@n2, B=@n4) {constB=0.5}) {texture=/DemonArenaGameplay/Assets/Textures/T_FogCloud}.R
   B=ScalarParameter() {defaultValue=2, parameterName=NoiseStrength}
  )
  n6 = ScalarParameter() {defaultValue=0.5, parameterName=InscatterOpacityMult_Morning}
  n7 = ScalarParameter() {defaultValue=1, parameterName=InscatterOpacityMult_Day}
  n8 = ScalarParameter() {defaultValue=0.85, parameterName=InscatterOpacityMult_Evening}
  n9 = ScalarParameter() {defaultValue=0.25, parameterName=InscatterOpacityMult_Night}
  n10 = ScalarParameter() {defaultValue=0.1, parameterName=ScatterOpacityMult_Morning}
  n11 = ScalarParameter() {defaultValue=0.6, parameterName=ScatterOpacityMult_Day}
  n12 = ScalarParameter() {defaultValue=0.25, parameterName=ScatterOpacityMult_Evening}
  n13 = ScalarParameter() {defaultValue=0.25, parameterName=ScatterOpacityMult_Night}
  n14 = Multiply(
   A=Saturate(None=Subtract(A=@n3, B=@n5))
   B=LinearInterpolate(
    A=MaterialFunctionCall(Value_Morning=@n6, Value_Day=@n7, Value_Evening=@n8, Value_Night=@n9) {materialFunction=&Scalar_Time_of_Day}.Result
    B=MaterialFunctionCall(Value_Morning=@n10, Value_Day=@n11, Value_Evening=@n12, Value_Night=@n13) {materialFunction=&Scalar_Time_of_Day}.Result
    Alpha=NamedRerouteUsage()
   )
  ) {constB=2}
  DepthFade(Opacity=Multiply(A=Saturate(None=@n14), B=NamedRerouteUsage())) {fadeDistanceDefault=512}
  n15 = Saturate(None=Power(Base=NamedRerouteUsage()))
  n16 = LinearInterpolate(
   A=LinearInterpolate(
    A=VectorParameter() {parameterName=FogDepthColor2, defaultValue=<rgba 0.1694 0.288 0.469 1>}.RGB
    B=VectorParameter() {parameterName=FogDepthColor, defaultValue=<rgba 0.3831 0.1872 0.468 1>}.RGB
    Alpha=@n15
   )
   B=LinearInterpolate(
    A=VectorParameter() {parameterName=FogColor2, defaultValue=<rgba 0.2902 0.4902 0.8 1>}.RGB
    B=VectorParameter() {parameterName=FogColor, defaultValue=<rgba 0.6549 0.3216 0.8 1>}.RGB
    Alpha=@n15
   )
   Alpha=NamedRerouteUsage()
  )
  n17 = LinearInterpolate(
   A=Multiply(A=VectorParameter() {parameterName=FogColor3, defaultValue=<rgba 0.1259 0.0949 0.1562 1>}.RGB, B=@n16)
   B=@n16
   Alpha=NamedRerouteUsage()
  )
  n18 = AtmosphericLightVector()
  LinearInterpolate(
   A=@n17
   B=Multiply(A=SkyLightEnvMapSample(Direction=@n18), B=@n17)
   Alpha=ScalarParameter() {defaultValue=0.2, parameterName=SkyLightColorBlend}
  )
  MaterialFunctionCall() {materialFunction=&ObjectPivotPoint}
  MaterialFunctionCall(X=PerInstanceCustomData!(DefaultValue) {dataIndex=2}, Y=PerInstanceCustomData() {dataIndex=3}, Z=PerInstanceCustomData() {dataIndex=4}) {materialFunction=&MakeFloat3}
  MaterialFunctionCall() {materialFunction=&WorldPositionBehindTranslucency}
  MaterialFunctionCall() {materialFunction=&ObjectPivotPoint}
  MaterialFunctionCall() {materialFunction=&ObjectPivotPoint}
  NamedRerouteDeclaration(None=TransformPosition(None=Constant3Vector()) {transformSourceType=TRANSFORMPOSSOURCE_Instance, transformType=TRANSFORMPOSSOURCE_World}) {name=ObjectPivotPosition, nodeColor=<rgba 0 0.5547 1 1>}
  n19 = MaterialFunctionCall(Float2=@n2) {materialFunction=&BreakOutFloat2Components}
  n20 = SphereMask(
   A=@n19.R
   B=ScalarParameter() {defaultValue=0.5, parameterName=HorizontalMask_SamplePos}
   Radius=ScalarParameter() {defaultValue=0.5, parameterName=HorizeontalMaskRadius}
   Hardness=ScalarParameter() {defaultValue=0.5, parameterName=HorizontalMask_Falloff}
  )
  n21 = SphereMask(
   A=@n19.G
   B=ScalarParameter() {defaultValue=1, parameterName=VerticalMask_SamplePos}
   Radius=ScalarParameter() {defaultValue=1, parameterName=VerticalMaskRadius}
   Hardness=ScalarParameter() {defaultValue=0.5, parameterName=VerticalMask_Falloff}
  )
  n22 = ScalarParameter() {parameterName=VerticalMask2_SamplePos}
  n23 = ScalarParameter() {defaultValue=1, parameterName=VerticalMask2Radius}
  n24 = ScalarParameter() {defaultValue=0.65, parameterName=VerticalMask2_Falloff}
  NamedRerouteDeclaration(None=Multiply(A=Multiply(A=@n20, B=@n21), B=SphereMask(A=@n19.G, B=@n22, Radius=@n23, Hardness=@n24))) {name=GradientMasks, nodeColor=<rgba 1 0 0.9844 1>}
  n25 = CameraPositionWS()
  n26 = Constant3Vector()
  n27 = Min(
   A=Distance(A=NamedRerouteUsage(), B=@n25)
   B=Distance(A=TransformPosition(None=@n26) {transformSourceType=TRANSFORMPOSSOURCE_Instance, transformType=TRANSFORMPOSSOURCE_World}, B=@n25)
  )
  n28 = ScalarParameter() {defaultValue=6000, parameterName=FadeDistance}
  n29 = Subtract(A=@n27, B=Add(A=@n28, B=ScalarParameter() {defaultValue=1024, parameterName=OpacityFadeDistanceOffset}))
  n30 = ScalarParameter() {defaultValue=2048, parameterName=FadeFalloffGradient}
  NamedRerouteDeclaration(None=Saturate(None=Divide(A=@n29, B=@n30))) {name=DistanceFade_Mask, nodeColor=<rgba 0.9219 0 1 1>}
  NamedRerouteDeclaration(None=Saturate(None=Subtract(A=@n27, B=@n28))) {name=DistanceFadeWPO_Mask, nodeColor=<rgba 0.5938 0 1 1>}
  n31 = ScalarParameter() {defaultValue=10000, parameterName=DepthDistance}
  NamedRerouteDeclaration(None=MaterialFunctionCall!(Distance Bias, Distance Divisor=@n31, Scene Depth, Pixel Depth) {materialFunction=/Game/Effects/Fort_Effects/Materials/_Functions/MF_DistanceBehindTranslucency_01}.Clamped) {name=FogDepthMask, nodeColor=<rgba 0.8281 1 0 1>}
  n32 = AtmosphericLightVector()
  n33 = Normalize(VectorInput=Subtract(A=NamedRerouteUsage(), B=WorldPosition().XYZ))
  n34 = Saturate(None=ConstantBiasScale(None=DotProduct(A=@n32, B=@n33)))
  n35 = ScalarParameter() {parameterName=SamplePos}
  n36 = ScalarParameter() {defaultValue=1, parameterName=TransmittanceRadius}
  n37 = ScalarParameter() {defaultValue=0.5, parameterName=TransmittanceHardness}
  NamedRerouteDeclaration(None=Power(Base=SphereMask(A=@n34, B=@n35, Radius=@n36, Hardness=@n37), Exp=ScalarParameter() {defaultValue=1, parameterName=TransmittanceFalloff})) {name=TransmittanceMask, nodeColor=<rgba 1 0 0.6328 1>}
  n38 = VectorParameter() {parameterName=IslandPosOffset, defaultValue=<rgba 1160 -244 3156 1>}
  NamedRerouteDeclaration(None=Add(A=ActorPositionWS(), B=Transform(None=@n38.RGB) {transformSourceType=TRANSFORMSOURCE_Local})) {name=IslandLocation, nodeColor=<rgba 0 0.9766 1 1>}
  n39 = NamedRerouteUsage()
  n40 = TextureCoordinate()
  n41 = PerInstanceCustomData() {constDefaultValue=5000}
  n42 = PerInstanceCustomData() {constDefaultValue=5000, dataIndex=1}
  n43 = ScalarParameter() {bUseCustomPrimitiveData=true, defaultValue=5000, parameterName=ScaleX}
  n44 = ScalarParameter() {bUseCustomPrimitiveData=true, defaultValue=5000, parameterName=ScaleY, primitiveDataIndex=1}
  n45 = LinearInterpolate(
   A=Constant2Vector()
   B=StaticSwitchParameter(True=AppendVector(A=@n41, B=@n42), False=AppendVector(A=@n43, B=@n44)) {defaultValue=true, parameterName=UseInstance}
   Alpha=NamedRerouteUsage()
  )
  n46 = Normalize(VectorInput=Subtract(A=CameraPositionWS(), B=NamedRerouteUsage()))
  n47 = ScalarParameter() {defaultValue=1200, parameterName=Cameraoffset}
  Reroute(None=Add(
    A=MaterialFunctionCall!(Center Location=@n39, 0-1 UVs=@n40, XY scale=@n45, OverridePivot, "Normals (Optional)") {materialFunction=&Sprite}.World position offset
    B=Multiply(A=@n46, B=@n47)
   ))
  Saturate(None=@n1)
  n48 = SceneDepth()
  n49 = Distance(A=ObjectPositionWS(), B=CameraPositionWS())
  Saturate(None=Divide(A=Add(A=@n48.R, B=@n49), B=@n31))
  ScalarParameter() {defaultValue=1, parameterName=Collapse_Debug}
 ]
}
