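# SilenceMoon_Dissolve is a library material function that turns a ramp map into dissolve masks.
# It scales RampMap by EdgeSharpness and subtracts a Dissolve-driven lerp from EdgeSharpness to -1, outputting a result (optionally clamped via static switch), Frac edge and soft edge.
# An Offset-shifted copy supplies a second set of three outputs.
# World-position noise and DebugTimeSine provide preview defaults; three SM_Dissolve parameters are unconnected.
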MaterialFunction /DemonArenaGameplay/Assets/VFX/SilenceMoon_Dissolve
graph {
 properties {bExposeToLibrary=true}
 nodes [
  n1 = Noise!(World Position, FilterWidth) {levels=3, scale=0.01}
  n2 = FunctionInput(Preview=Constant() {r=3}) {bUsePreviewValueAsDefault=true, inputName=EdgeSharpness, inputType=FunctionInput_Scalar, sortPriority=3}
  n3 = MaterialFunctionCall() {materialFunction=&DebugTimeSine}
  n4 = Subtract(
   A=Multiply(A=FunctionInput(Preview=@n1) {bUsePreviewValueAsDefault=true, inputName=RampMap, sortPriority=1}, B=@n2)
   B=LinearInterpolate(
    A=@n2
    Alpha=FunctionInput(Preview=@n3.Result) {bUsePreviewValueAsDefault=true, inputName=Dissolve, inputType=FunctionInput_Scalar, sortPriority=2}
   ) {constB=-1}
  )
  n5 = Saturate(None=@n4)
  n6 = FunctionInput(Preview=StaticBool() {value=true}) {bUsePreviewValueAsDefault=true, inputName=Clamped, inputType=FunctionInput_StaticBool, sortPriority=5}
  FunctionOutput(None=StaticSwitch(True=@n5, False=@n4, Value=@n6)) {outputName=DissolveResult, sortPriority=1}
  FunctionOutput(None=Frac(None=@n5)) {outputName=DissolveEdge, sortPriority=2}
  n7 = Abs(None=Add(A=@n4) {constB=-1})
  FunctionOutput(None=OneMinus(None=Saturate(None=@n7))) {outputName=DissolveSoftEdge, sortPriority=3}
  n8 = Add(
   A=@n4
   B=FunctionInput(Preview=Constant() {r=0.5}) {bUsePreviewValueAsDefault=true, inputName=Offset, inputType=FunctionInput_Scalar, sortPriority=4}
  )
  n9 = Saturate(None=@n8)
  FunctionOutput(None=StaticSwitch(True=@n9, False=@n8, Value=@n6)) {outputName=Offset_DissolveResult, sortPriority=4}
  FunctionOutput(None=Frac(None=@n9)) {outputName=Offset_DissolveEdge, sortPriority=5}
  n10 = Abs(None=Add(A=@n8) {constB=-1})
  FunctionOutput(None=OneMinus(None=Saturate(None=@n10))) {outputName=Offset_DissolveSoftEdge, sortPriority=6}
  ScalarParameter() {defaultValue=3, group=SM_Dissolve, parameterName=SM_Dissolve_EdgeSharp}
  ScalarParameter() {defaultValue=0.5, group=SM_Dissolve, parameterName=SM_Dissolve_Dissolve}
  ScalarParameter() {defaultValue=0.2, group=SM_Dissolve, parameterName=SM_Dissolve_Offset}
 ]
}
