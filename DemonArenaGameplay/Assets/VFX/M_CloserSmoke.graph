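# M_CloserSmoke is an unlit, translucent, two-sided smoke material for Niagara mesh particles, drawn without depth testing.
# It derives camera-distance masks around a Radius parameter, plus view-facing and camera-space Fresnel terms.
# Tilted, panned UVs, a tiling texture and a dissolve function form the opacity mask.
# Two texture-driven color pairs blend into the base color, alongside time-of-day, emissive-adjustment and camera-offset functions.
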Material /DemonArenaGameplay/Assets/VFX/M_CloserSmoke
graph {
 properties {
  bDisableDepthTest=true
  bUsedWithNiagaraMeshParticles=true
  blendMode=BLEND_Translucent
  pixelDepthOffsetMode=PDOM_Legacy
  shadingModel=MSM_Unlit
  twoSided=true
  usedShadingModels=Unlit
 }
 nodes [
  ActorPositionWS()
  ActorPositionWS()
  CameraVectorWS()
  n1 = ObjectPositionWS()
  n2 = CameraPositionWS()
  n3 = Transform(None=Constant3Vector() {constant=<rgba 0 0 1 0>}) {transformSourceType=TRANSFORMSOURCE_Camera}
  n4 = DotProduct(A=Normalize(VectorInput=Subtract(A=@n1, B=@n2)), B=@n3)
  ConstantBiasScale(None=@n4)
  If(A=NamedRerouteUsage(), "A > B"=Subtract(A=CameraPositionWS(), B=ObjectPositionWS()), "A < B"=Constant() {r=-9.999999778196308e+22})
  LinearInterpolate(
   A=VectorParameter() {parameterName=Param_4, defaultValue=<rgba 0 1 0 0>}.RGB
   B=VectorParameter() {parameterName=Param_3, defaultValue=<rgba 1 0 0 0>}.RGB
   Alpha=NamedRerouteUsage()
  )
  n5 = DotProduct(A=VertexNormalWS(), B=@n3)
  n6 = OneMinus(None=@n5)
  LinearInterpolate(A=@n6, B=Abs(None=@n5), Alpha=ScalarParameter() {parameterName="Fresnel_Invert (0or1)"})
  LinearInterpolate()
  MaterialFunctionCall() {materialFunction=&Color_Time_of_Day}
  n7 = NamedRerouteUsage()
  MaterialFunctionCall(Emissive=@n7) {materialFunction=&MF_FX_EmissiveAdjustments}
  MaterialFunctionCall(Offset Amount=ScalarParameter() {defaultValue=999, parameterName=Param_4}) {materialFunction=/Engine/Functions/Engine_MaterialFunctions02/WorldPositionOffset/CameraOffset}
  Multiply(A=@n7, B=NamedRerouteUsage())
  n8 = Power(Base=OneMinus(None=NamedRerouteUsage()), Exp=ScalarParameter() {defaultValue=5, parameterName=Param_1})
  n9 = NamedRerouteUsage()
  Multiply(A=Multiply(A=@n7, B=Multiply(A=@n8, B=@n9)), B=VectorParameter() {defaultValue=<rgba 1 0.0174 0.0174 0>}.RGB)
  n10 = ConstantBiasScale(None=NamedRerouteUsage())
  n11 = NamedRerouteUsage()
  Multiply(A=Add(A=@n10, B=@n11)) {constB=0.5}
  n12 = ScalarParameter() {defaultValue=800, parameterName=Radius}
  n13 = Subtract(A=Distance(A=CameraPositionWS(), B=ObjectPositionWS()), B=@n12)
  n14 = Constant() {r=1}
  n15 = Constant()
  NamedRerouteDeclaration(None=MaterialFunctionCall(
    Input=Abs(None=@n13)
    Input Low=Constant()
    Input High=ScalarParameter() {defaultValue=500, parameterName=Param_3}
    Target Low=@n14
    Target High=@n15
   ) {materialFunction=&RemapValueRange}.Result) {name=RangeValue, nodeColor=<rgba 1 0.7969 0 1>}
  n16 = MaterialFunctionCall(Float2=TextureCoordinate()) {materialFunction=&BreakOutFloat2Components}
  n17 = Add(A=@n16.G, B=Multiply(A=@n16.R, B=ScalarParameter() {defaultValue=-1, parameterName=UV_Tilt}))
  n18 = VectorParameter() {parameterName=Main_UV, defaultValue=<rgba 2 2 -0.5387 0>}
  NamedRerouteDeclaration(None=MaterialFunctionCall!(
    ClampTexture
    UV=MaterialFunctionCall(X=@n16.R, Y=@n17) {materialFunction=&MakeFloat2}.Result
    Tex
    InRGB=@n18.RGB
    InA=@n18.RGB
    Time
    TimeMult
   ) {materialFunction=&/MF_MOON_TexPanner_V2}.PannerUV) {name=MainUV, nodeColor=<rgba 0 0.4049 1 1>}
  NamedRerouteDeclaration(None=MaterialFunctionCall(Input=@n13, Input Low=Multiply(A=@n12) {constB=0.5}, Input High=Multiply(A=@n12) {constB=-0.5}, Target Low=@n15, Target High=@n14) {materialFunction=&RemapValueRange}.Result) {name=RangeValueLinear, nodeColor=<rgba 0.4297 0 1 1>}
  n19 = NamedRerouteUsage()
  n20 = NamedRerouteUsage()
  n21 = LinearInterpolate(B=ScalarParameter() {defaultValue=1.1}, Alpha=NamedRerouteUsage())
  NamedRerouteDeclaration(None=MaterialFunctionCall!(
    RampMap=TextureSampleParameter2D(UVs=@n19) {parameterName=Param_1, samplerType=SAMPLERTYPE_LinearColor, texture=&/Tile_output}.G
    Dissolve=Add(A=@n20, B=@n21)
    EdgeSharpness=ScalarParameter() {defaultValue=2, parameterName=Param_9}
    Offset=ScalarParameter() {defaultValue=1.1748, parameterName=Param_10}
    Clamped
   ) {materialFunction=&/SilenceMoon_Dissolve}.DissolveResult) {name=Opacity, nodeColor=<rgba 0.1953 0 1 1>}
  NamedRerouteDeclaration(None=@n4) {name=FacingFresnel, nodeColor=<rgba 1 0 0.5859 1>}
  n22 = VectorParameter() {parameterName=Param_7, defaultValue=<rgba 0.0192 0.0129 0.026 1>}
  n23 = VectorParameter() {parameterName=Param_8, defaultValue=<rgba 0.0806 0.0863 0.1719 1>}
  n24 = TextureSampleParameter2D(UVs=NamedRerouteUsage()) {parameterName=Param, samplerType=SAMPLERTYPE_LinearColor, texture=&Tile_output}
  n25 = VectorParameter() {parameterName=Param_6, defaultValue=<rgba 0.5469 0.0028 0.0028 1>}
  n26 = VectorParameter() {parameterName=Param_5, defaultValue=<rgba 0 0 0 1>}
  n27 = TextureSampleParameter2D(UVs=NamedRerouteUsage()) {parameterName=Param, samplerType=SAMPLERTYPE_LinearColor, texture=&Tile_output}
  NamedRerouteDeclaration(None=LinearInterpolate(
    A=LinearInterpolate(A=@n22.RGB, B=@n23.RGB, Alpha=@n24.G)
    B=LinearInterpolate(A=@n25.RGB, B=@n26.RGB, Alpha=@n27.G)
    Alpha=LinearInterpolate(B=@n10, Alpha=@n11)
   )) {name=BaseColor, nodeColor=<rgba 1 0.9844 0 1>}
  NamedRerouteDeclaration(None=Power(Base=@n6, Exp=ScalarParameter() {defaultValue=1, parameterName=Fresnel_Power})) {name=Fresnel, nodeColor=<rgba 1 0 0.9613 1>}
  NamedRerouteUsage()
  NamedRerouteUsage()
  Normalize(VectorInput=Subtract(A=CameraPositionWS(), B=WorldPosition().XYZ))
  ObjectPositionWS()
  Reroute(None=@n5)
  Saturate(None=@n4)
 ]
}
