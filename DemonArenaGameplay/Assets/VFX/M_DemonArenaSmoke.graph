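# Two-sided unlit translucent smoke material for the demon arena VFX.
# Object-position-offset time pans swizzlable UVs for fractal smoke textures, with bump-offset parallax and a vertex-alpha-driven dissolve for opacity.
# Color combines a contrast-ramped emissive core with time-of-day tinted fake lighting, alongside sun-direction, Fresnel, sine-mask and depth-fade terms.
# Vertex offset math rotates the mesh around its pivot, adding a previous-frame-only normal push that fakes motion vectors.
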Material /DemonArenaGameplay/Assets/VFX/M_DemonArenaSmoke
graph {
 properties {
  blendMode=BLEND_Translucent
  pixelDepthOffsetMode=PDOM_Legacy
  shadingModel=MSM_Unlit
  translucencyLightingMode=TLM_VolumetricDirectional
  twoSided=true
  usedShadingModels=Unlit
 }
 nodes [
  n1 = VertexNormalWS()
  n2 = Multiply(A=ScalarParameter() {defaultValue=30.7052, parameterName=Param_6}, B=OneMinus(None=NamedRerouteUsage()))
  n3 = Time()
  n4 = ScalarParameter() {defaultValue=-0.03, parameterName=Param_1}
  Add(
   A=PreviousFrameSwitch(Current Frame=Constant(), Previous Frame=Multiply(A=@n1, B=@n2))
   B=RotateAboutAxis(
    NormalizedRotationAxis=VectorParameter() {parameterName=Param_3, defaultValue=<rgba 0 0 1 0>}.RGB
    RotationAngle=Multiply(A=@n3, B=@n4)
    PivotPoint=ObjectPositionWS()
    Position=WorldPosition().XYZ
   )
  )
  Constant() {r=1}
  MaterialFunctionCall(Emissive=Add(A=NamedRerouteUsage(), B=NamedRerouteUsage())) {materialFunction=&MF_FX_EmissiveAdjustments}
  Multiply(A=Multiply(A=NamedRerouteUsage(), B=NamedRerouteUsage()), B=NamedRerouteUsage())
  n5 = Subtract(A=WorldPosition().XYZ, B=ActorPositionWS())
  n6 = VertexInterpolator(VS=DotProduct(A=SkyAtmosphereLightDirection(), B=Normalize(VectorInput=@n5)))
  NamedRerouteDeclaration(None=Saturate(None=ConstantBiasScale(None=@n6.PS))) {name=LightSource, nodeColor=<rgba 0.9453 1 0 1>}
  n7 = TextureCoordinate()
  n8 = MaterialFunctionCall(XY=@n7) {materialFunction=&Swizzle}
  n9 = ScalarParameter() {defaultValue=1, parameterName=Main_UV_Swizzle}
  n10 = VectorParameter() {parameterName=Main_UV, defaultValue=<rgba 0.4272 1.5912 0.0194 0.0583>}
  n11 = ObjectPositionWS()
  n12 = Add(
   A=Divide(A=MaterialFunctionCall(f3=@n11) {materialFunction=&AddComponents}.f3) {constB=111}
   B=Time()
  )
  NamedRerouteDeclaration(None=MaterialFunctionCall!(ClampTexture, UV=LinearInterpolate(A=@n7, B=@n8.YX, Alpha=@n9), Tex, InRGB=@n10.RGB, InA=@n10.RGB, Time=@n12, TimeMult) {materialFunction=&/MF_MOON_TexPanner_V2}.PannerUV) {name=MainUV, nodeColor=<rgba 0 0.4049 1 1>}
  n13 = ScalarParameter() {defaultValue=3, parameterName=s}
  n14 = ScalarParameter() {defaultValue=1.9537, parameterName=Param_3}
  n15 = TextureSample(UVs=NamedRerouteUsage()) {texture=&/Tile_FractalSmoke_Channel}
  n16 = VertexColor()
  n17 = ScalarParameter() {defaultValue=1, parameterName=Param_2}
  n18 = NamedRerouteUsage()
  n19 = Multiply(
   A=Multiply(A=NamedRerouteUsage(), B=Power(Base=@n16.A, Exp=@n17)) {constA=1}
   B=LinearInterpolate(
    A=TextureSample(UVs=@n18) {texture=&Tile_FractalSmoke_Channel}.G
    Alpha=ScalarParameter() {defaultValue=0.3057, parameterName=Param_7}
   )
  )
  n20 = Multiply(A=VertexColor().A, B=ScalarParameter() {defaultValue=2.5548, parameterName=Param_11})
  NamedRerouteDeclaration(None=Multiply(
    A=MaterialFunctionCall!(RampMap=@n15.G, Dissolve=@n19, EdgeSharpness=@n13, Offset=@n14, Clamped) {materialFunction=&/SilenceMoon_Dissolve}.Offset_DissolveResult
    B=Saturate(None=@n20)
   )) {name=Opacity, nodeColor=<rgba 1 0.9844 0 1>}
  n21 = ScalarParameter() {defaultValue=500, parameterName=Param_16}
  n22 = ScalarParameter() {defaultValue=100, parameterName=Param_15}
  NamedRerouteDeclaration(None=DepthFade(Opacity=NamedRerouteUsage(), FadeDistance=LinearInterpolate(A=@n21, B=@n22))) {name=DepthFadeResult, nodeColor=<rgba 1 0.2344 0 1>}
  n23 = LinearInterpolate(
   A=NamedRerouteUsage()
   B=TextureSample(UVs=NamedRerouteUsage()) {texture=&Tile_FractalSmoke_Channel}.G
   Alpha=ScalarParameter() {defaultValue=0.8, parameterName=Param_8}
  )
  n24 = ScalarParameter() {defaultValue=-0.168, parameterName=Param_4}
  NamedRerouteDeclaration(None=BumpOffset(Coordinate=NamedRerouteUsage(), Height=@n23, HeightRatioInput=@n24) {referencePlane=0}) {name=BumpUV, nodeColor=<rgba 0.7344 0 1 1>}
  NamedRerouteDeclaration(None=BumpOffset(Coordinate=NamedRerouteUsage(), Height=@n23, HeightRatioInput=@n24) {referencePlane=0}) {name=BumpUV1, nodeColor=<rgba 1 0 0.6094 1>}
  n25 = ScalarParameter() {defaultValue=3}
  n26 = VectorParameter() {defaultValue=<rgba 0 0 0 1>}
  n27 = VectorParameter() {parameterName=Param_4, defaultValue=<rgba 5 0 0 1>}
  n28 = MaterialFunctionCall(In=@n15.R, Contrast=ScalarParameter() {defaultValue=0.2, parameterName=Param_12}) {materialFunction=&CheapContrast}
  NamedRerouteDeclaration(None=Multiply(A=Power(Base=@n15.R, Exp=@n25), B=LinearInterpolate(A=@n26.RGB, B=@n27.RGB, Alpha=@n28.Result))) {name=CoreEmissive, nodeColor=<rgba 0 1 0.3906 1>}
  n29 = MaterialFunctionCall() {materialFunction=&Color_Time_of_Day}
  n30 = ScalarParameter() {defaultValue=0.5, parameterName=Param_9}
  n31 = VectorParameter() {parameterName=Param_6, defaultValue=<rgba 0.0156 0.0156 0.0156 1>}
  n32 = NamedRerouteUsage()
  n33 = TextureCoordinate()
  n34 = LinearInterpolate(
   A=VectorParameter() {parameterName=Param_1, defaultValue=<rgba 0.1238 0.1238 0.3333 1>}.RGB
   B=Multiply(
    A=TextureSampleParameter2D(UVs=@n32) {parameterName=Color, texture=&/Tile_output_3}.RGB
    B=VectorParameter() {parameterName=Param_2, defaultValue=<rgba 0.1771 0.0277 0.0548 1>}.RGB
   )
   Alpha=LinearInterpolate(A=ComponentMask(None=@n33) {g=true}, B=@n15.G, Alpha=ScalarParameter() {defaultValue=1.5947, parameterName=ShadowBalance})
  )
  n35 = NamedRerouteUsage()
  n36 = ScalarParameter() {defaultValue=0.7757, parameterName=Param_5}
  n37 = NamedRerouteUsage()
  NamedRerouteDeclaration(None=LinearInterpolate(
    A=Multiply(A=@n29.Result, B=@n30)
    B=LinearInterpolate(A=@n31.RGB, B=@n34, Alpha=@n35)
    Alpha=LinearInterpolate(A=@n36, Alpha=@n37) {constA=0.5}
   )) {name=GI, nodeColor=<rgba 0 1 0.0625 1>}
  n38 = TextureCoordinate() {uTiling=0.5, vTiling=0.5}
  n39 = TextureCoordinate() {coordinateIndex=1, uTiling=0.5, vTiling=0.5}
  n40 = ScalarParameter() {parameterName=CubeMap_Switch}
  n41 = MaterialFunctionCall(Float2=Sine(None=LinearInterpolate(A=@n38, B=@n39, Alpha=@n40))) {materialFunction=&BreakOutFloat2Components}
  n42 = MaterialFunctionCall(A=@n41.R, B=Multiply(A=@n41.R, B=@n41.R), C=@n41.R, Alpha=ScalarParameter() {parameterName=CubeMap_Select}) {materialFunction=&Lerp_3Color}
  n43 = ScalarParameter() {defaultValue=5, parameterName=Param_14}
  NamedRerouteDeclaration(None=Saturate(None=Multiply(A=@n42.Result, B=@n43))) {name=CubeMap, nodeColor=<rgba 1 1 1 1>}
  n44 = Subtract(A=CameraPositionWS(), B=WorldPosition().XYZ)
  n45 = Abs(None=DotProduct(A=VertexNormalWS(), B=Normalize(VectorInput=@n44)))
  n46 = Power(
   Base=LinearInterpolate(A=OneMinus(None=@n45), B=@n45, Alpha=ScalarParameter() {defaultValue=1, parameterName="Fresnel_Invert (0or1)"})
   Exp=ScalarParameter() {defaultValue=1.83, parameterName=Fresnel_Power}
  )
  n47 = ScalarParameter() {defaultValue=0.05, parameterName=Fresnel_Offset}
  NamedRerouteDeclaration(None=Multiply(A=Subtract(A=@n46, B=@n47), B=Add(A=@n47))) {name=Fresnel, nodeColor=<rgba 1 0 0.9613 1>}
  n48 = TextureCoordinate()
  n49 = MaterialFunctionCall(XY=@n48) {materialFunction=&Swizzle}
  n50 = ScalarParameter() {defaultValue=1, parameterName=UV1_Swizzle}
  n51 = VectorParameter() {parameterName=UV1, defaultValue=<rgba 0.6569 4.9882 0.0095 0.0258>}
  NamedRerouteDeclaration(None=MaterialFunctionCall(UV=LinearInterpolate(A=@n48, B=@n49.YX, Alpha=@n50), InRGB=@n51.RGB, InA=@n51.RGB, Time=@n12) {materialFunction=&MF_MOON_TexPanner_V2}.PannerUV) {name=UV_1, nodeColor=<rgba 0 0.4049 1 1>}
  NamedRerouteUsage()
 ]
}
