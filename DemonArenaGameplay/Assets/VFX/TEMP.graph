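# TEMP is a minimal two-sided Default Lit material in the VFX folder, apparently a placeholder.
# Its graph holds only a scalar constant and a dark grey vector parameter, with no connections shown between them or to material outputs.
# Automatic usage flag setting in the editor is turned off.
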Material /DemonArenaGameplay/Assets/VFX/TEMP
graph {
 properties {bAutomaticallySetUsageInEditor=false, twoSided=true, usedShadingModels=Default Lit}
 nodes [
  Constant() {r=0.8}
  VectorParameter() {defaultValue=<rgba 0.0417 0.0417 0.0417 0>}
 ]
}
